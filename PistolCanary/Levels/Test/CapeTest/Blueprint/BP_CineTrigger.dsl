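; BP_CineTrigger is a trigger actor in a CapeTest test level that plays its assigned Level Sequence when overlapped.
; Any overlap on the Trigger component hides it, spawns a Level Sequence player, plays it and binds its finished event.
; The finished event makes the Trigger visible again; only visibility is toggled, not collision.
; The Construction Script labels the TextRender with "Play:" and the sequence name.
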(event EventBeginPlay)

(event Collision|EventActorBeginOverlap
  (params (OtherActor :type Actor%20Object%20Reference)))

(event EventTick
  (params (DeltaSeconds :type Float%20%28single-precision%29)))

(event OnComponentBeginOverlap(Trigger)
  (params (OverlappedComponent :type Primitive%20Component%20Object%20Reference) (OtherActor :type Actor%20Object%20Reference) (OtherComp :type Primitive%20Component%20Object%20Reference) (OtherBodyIndex :type Integer) (bFromSweep :type Boolean) (SweepResult :type Hit%20Result%20Structure%20%28by%20ref%29))
  (Rendering|SetVisibility (Variables|Default|GetTrigger) false false)
  (bind _outactor (Sequencer|Player|CreateLevelSequencePlayer (Variables|Default|GetLevelSequence) (struct :bAutoPlay true :LoopCount "(Value=0)" :TickInterval "(TickIntervalSeconds=0.000000,EvaluationBudgetMicroseconds=0.000000,bTickWhenPaused=False,bAllowRounding=True)" :PlayRate 1.0 :StartTime 0.0 :bRandomStartTime false :bDisableMovementInput false :bDisableLookAtInput false :bHidePlayer false :bHideHud false :bDisableCameraCuts false :FinishCompletionStateOverride "None" :bPauseAtEnd false)))
  (Sequencer|Player|Play (Variables|/Script/LevelSequence.LevelSequenceActor|GetSequencePlayer _outactor) :!class /Script/MovieScene.MovieSceneSequencePlayer)
  (Sequencer|Player|AssignOnFinished (Variables|/Script/LevelSequence.LevelSequenceActor|GetSequencePlayer _outactor) (delegate Custom|OnFinished_Event)))

(event Custom|OnFinished_Event
  (Rendering|SetVisibility (Variables|Default|GetTrigger) true false))

(fn ConstructionScript
  (Rendering|Components|TextRender|SetText (Variables|Default|GetTextRender) (Utilities|Text|ToText(String) (Utilities|String|Append "Play:" " " (Utilities|String|ToString(Text) (Utilities|Text|ToText(Object) (Variables|Default|GetLevelSequence)))))))
